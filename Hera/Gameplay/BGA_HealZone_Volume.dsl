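; Heal zone volume actor that gives Fortnite player pawns the Hera tide pool healing effect while they are inside it.
; On overlap begin, it applies the healing-grant GameplayEffect as a global environment effect to any overlapping player pawn.
; On overlap end, it removes all stacks of that effect through the pawn's ability system component, if valid.
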(event EventBeginPlay
  (|CallParentFunction))

(event OnComponentBeginOverlap(OverlapVolume)
  (params (OverlappedComponent :type Primitive%20Component%20Object%20Reference) (OtherActor :type Actor%20Object%20Reference) (OtherComp :type Primitive%20Component%20Object%20Reference) (OtherBodyIndex :type Integer) (bFromSweep :type Boolean) (SweepResult :type Hit%20Result%20Structure%20%28by%20ref%29))
  (Utilities|IsA(soft) OtherActor "/Script/FortniteGame.FortPlayerPawnAthena"
    (:True
      (Environment|ApplyGlobalEnvironmentGameplayEffecttoActor OtherActor "/Hera/Gameplay/TidePools/GE_Hera_TidePool_EnvHeal_Grant_Healing.GE_Hera_TidePool_EnvHeal_Grant_Healing_C"))))

(event OnComponentEndOverlap(OverlapVolume)
  (params (OverlappedComponent :type Primitive%20Component%20Object%20Reference) (OtherActor :type Actor%20Object%20Reference) (OtherComp :type Primitive%20Component%20Object%20Reference) (OtherBodyIndex :type Integer))
  (bind _returnvalue (Ability|GetFortAbilitySystemComponent OtherActor))
  (Utilities|IsA(soft) OtherActor "/Script/FortniteGame.FortPlayerPawnAthena"
    (:True
      (Utilities|IsValid _returnvalue
        (:Is%20Valid
          (GameplayEffects|RemoveActiveGameplayEffectbySourceEffect _returnvalue "/Hera/Gameplay/TidePools/GE_Hera_TidePool_EnvHeal_Grant_Healing.GE_Hera_TidePool_EnvHeal_Grant_Healing_C" 0 -1))))))

(fn ConstructionScript)
